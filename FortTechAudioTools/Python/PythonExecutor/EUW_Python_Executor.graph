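# Editor utility widget for FortTechAudioTools that lists, describes and runs Python scripts from the plugin's PythonExecutor folder inside the editor.
# Refresh rglobs .py files beside the widget's Blueprint into ScriptArray, with GetScriptOptions returning that list.
# Changing ScriptToRun in a Single Property View shows the script's get_metadata() author and description.
# Run imports and reloads the module, calling execute(args) with Arguments text, which Capture fills from selected asset paths.
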EditorUtilityWidgetBlueprint /FortTechAudioTools/Python/PythonExecutor/EUW_Python_Executor
graphs [
 EventGraph {nodes [
   "On Clicked (Button_Run)"() {is=title K2Node_ComponentBoundEvent}
   n1 = Get Arguments() -> (Arguments:Editor Utility Editable Text Box Object Reference)
   > Execute Python Script!(
    execute
    PythonScript=Get Script_Execute() -> (Script_Execute:String)
    script_name:String=Get ScriptToRun() -> (ScriptToRun:String)
    args:String="To String (Text)"(InText=Get Text(self=@n1))
   ) -> (then, ReturnValue) {inputs [script_name,args]}
   n2 = "On Clicked (Button_Refresh)"() {is=title K2Node_ComponentBoundEvent}
   Clear(execute=[@n2.then, @n14.then], TargetArray=Get ScriptArray() -> (ScriptArray:Array of Strings))
   > n3 = Execute Python Script!(
    execute
    PythonScript="from pathlib import Path\r\n\r\nlist = []\r\nfor file_path in Path(path).rglob(f\"*.py\"):\r\n    list.append(file_path.name)"
    path:String=GetSystemPath!() -> (ReturnValue:String)
   ) -> (then, ReturnValue, list:Array of Strings) {inputs [path], outputs [list]}
   > n4 = For Each Loop(Array=@n3.list)
   > Branch(Condition=AND Boolean(A=Contains(SearchIn=@n4, Substring=.py), B="Not Equal Exactly (String)"(A=@n4, B=__init__.py)))
   > Add(TargetArray=Get ScriptArray(), NewItem=Replace(SourceString=@n4, From=.py)) {is=title K2Node_CallArrayFunction}
   "On Clicked (Button_Clear)"() {is=title K2Node_ComponentBoundEvent}
   > Set Text(self:Text Box Object Reference=Get Arguments())
   n5 = "On Property Changed (ScriptToRun_Display)"!() -> (OutputDelegate, then, PropertyName:Name=None) {is=title K2Node_ComponentBoundEvent}
   > Branch(Condition="Equal (Name)"(A=@n5.PropertyName, B=ScriptToRun))
   > n6 = Execute Python Script!(execute, PythonScript=Get Script_GetMetadata() -> (Script_GetMetadata:String), script_name:String=Get ScriptToRun()) -> (then, ReturnValue, result:Array of Strings) {inputs [script_name], outputs [result]}
   n7 = "Get (a ref)"(Array:Array of Strings=@n6.result, Dimension 1=1) -> (Output:"String (by ref)")
   n8 = "Get (a ref)"(Array=@n6.result, Dimension 1=0)
   > "SetText (Multi-Line Text Box)"!(
    execute
    self:"Text Box (Multi- Line) Object Reference"=Get Metadata() -> (Metadata:Editor Utility Multi Line Editable Text Box Object Reference)
    InText:Text=Format Text!(Format="LOCTEXT(\"{Author}\\r\\n{Description}\")", Author:"String (by ref)"=@n8, Description:"String (by ref)"=@n7) -> (Result)
   ) -> (then)
   > Set ScriptTooltip(ScriptTooltip:String=Append!(A=@n8, B="\r\n", C:String=@n7) -> (ReturnValue)) -> (Output_Get:String)
   "On Clicked (Button_Capture)"() {is=title K2Node_ComponentBoundEvent}
   > n9 = GetSelectedAssets()
   > n10 = For Each Loop(Array:Array of Object References=@n9) -> (Array Element:Object Reference)
   n11 = "Not Equal (Integer)"(A=@n10.Array Index)
   n12 = Get Object Path String(Object=@n10)
   n13 = Get Arguments()
   > "SetText (Text Box)"!(
    execute
    self:Text Box Object Reference=Get Arguments()
    InText:Text=Format Text!(
     Format="LOCTEXT(\"{prev}{sep}{new}\")"
     prev:Text=Get Text(self=@n13)
     sep:String=SelectString(A=", ", bPickA=@n11)
     new:String=Split(SourceString=@n12, InStr=., SearchDir=FromEnd)
    ) -> (Result)
   ) -> (then)
   "On Clicked (Button_Docs)"() {is=title K2Node_ComponentBoundEvent}
   > LaunchURL!(execute, URL:String=Get DocumentationURL() -> (DocumentationURL:String)) -> (then)
   n14 = Refresh() {is=title K2Node_CustomEvent}
   Event Construct()
   > SetObject(self=Get ScriptToRun_Display() -> (ScriptToRun_Display:Single Property View Object Reference), NewObject=Self-Reference())
   > Refresh() {is=title K2Node_CallFunction}
  ]}
 GetScriptOptions {nodes [
   GetScriptOptions() {is=title K2Node_FunctionEntry}
   > Return Node!(execute, Options:Array of Strings=Get ScriptArray())
  ]}
 GetSystemPath {nodes [
   GetSystemPath() {is=title K2Node_FunctionEntry}
   > n1 = GetBlueprintForClass!(execute, Class:Object Class Reference=Get Class(Object=Self-Reference()) -> (ReturnValue:EUW Python Executor Class Reference)) -> (True:Exec, False:Exec, ReturnValue:Blueprint Object Reference)
   Return Node!(
    execute=[@n1.False, @n1.True]
    ReturnValue:String=GetPath!(InPath:String=GetSystemPath!(Object:Object Reference=@n1) -> (ReturnValue) {is=title K2Node_CallFunction}) -> (ReturnValue:String)
   )
  ]}
]
variables [
 DocumentationURL {type=string, default=https://epicgames.atlassian.net/wiki/spaces/FORT/pages/451085746/Python+Executor+Tool}
 NewVar {type=boolean}
 Script_Execute {
  type=string
  default="import sys\r\nfrom pathlib import Path\r\nimport importlib\r\nimport unreal\r\n\r\n## Get the plugin's content directory\r\nplugin_content_dir = unreal.Paths.project_plugins_dir() + \"FortTechAudioTools/Content/Python/PythonExecutor\"\r\nscript_dir = Path(unreal.Paths.convert_relative_path_to_full(plugin_content_dir))\r\n\r\nif str(script_dir) not in sys.path:\r\n  sys.path.insert(0, str(script_dir))\r\n\r\n## Import and reload to get latest changes\r\nscript_module = importlib.import_module(script_name)\r\nimportlib.reload(script_module)\r\n\r\n## Call the function\r\nresult = script_module.execute(args)"
 }
 Script_Execute_test {
  type=string
  default="import sys\r\nfrom pathlib import Path\r\nimport importlib\r\nimport unreal\r\n\r\n## Get the plugin's content directory\r\nplugin_content_dir = unreal.Paths.project_plugins_dir() + \"FortTechAudioTools/Content/Python/BulkAssetEditor\"\r\nscript_dir = Path(unreal.Paths.convert_relative_path_to_full(plugin_content_dir))\r\n\r\nif str(script_dir) not in sys.path:\r\n  sys.path.insert(0, str(script_dir))\r\n\r\n## Import and reload to get latest changes\r\nscript_module = importlib.import_module(script_name)\r\nimportlib.reload(script_module)\r\n\r\n## Call the function\r\nresult = script_module.execute(args)"
 }
 Script_GetMetadata {
  type=string
  default="import sys\r\nfrom pathlib import Path\r\nimport importlib\r\nimport unreal\r\n\r\n## Get the plugin's content directory\r\nplugin_content_dir = unreal.Paths.project_plugins_dir() + \"FortTechAudioTools/Content/Python/PythonExecutor\"\r\nscript_dir = Path(unreal.Paths.convert_relative_path_to_full(plugin_content_dir))\r\n\r\nif str(script_dir) not in sys.path:\r\n  sys.path.insert(0, str(script_dir))\r\n\r\n## Import and reload to get latest changes\r\nscript_module = importlib.import_module(script_name)\r\nimportlib.reload(script_module)\r\n\r\n## Call the function\r\nresult = script_module.get_metadata()"
 }
 ScriptArray {type="array<string>", default []}
 ScriptTooltip {type=string, default=""}
 ScriptToRun {type=string, default=""}
]
widgets {
 info {rootWidgetClass=&VerticalBox, widgetCount=19}
 widgets [
  VerticalBox_215 {widgetClassPath=&VerticalBox, widget=VerticalBox_215, properties {bOverride_Cursor=true}}
  Spacer_76 {
   widgetClassPath=&Spacer
   parent=VerticalBox_215
   widget=Spacer_76
   slot {class=VerticalBoxSlot, properties {padding=<ltrb 5 5 5 5>}}
  }
  HorizontalBox_2 {
   widgetClassPath=&HorizontalBox
   parent=VerticalBox_215
   widget=HorizontalBox_2
   slot {class=VerticalBoxSlot, properties {verticalAlignment=VAlign_Top}}
  }
  Button_Docs {
   widgetClassPath=&EditorUtilityButton
   parent=HorizontalBox_2
   widget=Button_Docs
   properties {toolTipText=Opens up Confluence page documentation}
   slot {class=HorizontalBoxSlot, properties {horizontalAlignment=HAlign_Left, padding=<ltrb 5 5 2.5 5>}}
  }
  TechAudioToolsEditorSlateIcon_1 {
   widgetClassPath=/Script/TechAudioToolsContentEditor.TechAudioToolsEditorSlateIcon
   bIsVariable=true
   parent=Button_Docs
   widget=TechAudioToolsEditorSlateIcon_1
   properties {fullIconPath=EditorStyle.Icons.Documentation}
   slot {class=ButtonSlot, properties {horizontalAlignment=HAlign_Fill, padding=<ltrb 0 0 0 0>}}
  }
  Button_Refresh {
   widgetClassPath=&EditorUtilityButton
   parent=HorizontalBox_2
   widget=Button_Refresh
   properties {toolTipText=Refreshes the list of scripts to choose from}
   slot {class=HorizontalBoxSlot, properties {horizontalAlignment=HAlign_Left, padding=<ltrb 2.5 5 2.5 5>}}
  }
  TextBlock {
   widgetClassPath=&TextBlock
   parent=Button_Refresh
   widget=TextBlock
   properties {
    text=Refresh
    font {
     bForceMonospaced=false
     bMaterialIsStencil=false
     fontMaterial=null
     fontObject=&Roboto
     letterSpacing=0
     monospacedWidth=1
     size=13.5
     skewAmount=0
     typefaceFontName=Regular
     outlineSettings {
      bApplyOutlineToDropShadows=false
      bMiteredCorners=false
      bSeparateFillAlpha=false
      outlineMaterial=null
      outlineSize=0
      outlineColor=<rgba 0 0 0 1>
     }
    }
   }
   slot {class=ButtonSlot}
  }
  ScriptToRun_Display {
   widgetClassPath=&SinglePropertyView
   parent=HorizontalBox_2
   widget=ScriptToRun_Display
   properties {propertyName=ScriptToRun}
   slot {class=HorizontalBoxSlot, properties {padding=<ltrb 2.5 5 5 5>, size {sizeRule=Fill, value=1}}}
  }
  HorizontalBox_0 {
   widgetClassPath=&HorizontalBox
   parent=VerticalBox_215
   widget=HorizontalBox_0
   slot {class=VerticalBoxSlot, properties {verticalAlignment=VAlign_Top}}
  }
  Button_Run {
   widgetClassPath=&EditorUtilityButton
   parent=HorizontalBox_0
   widget=Button_Run
   properties {toolTipText=Executes the currently-selected script with any supplied Arguments}
   slot {class=HorizontalBoxSlot, properties {horizontalAlignment=HAlign_Left, verticalAlignment=VAlign_Top, padding=<ltrb 5 5 2.5 5>}}
  }
  TextBlock_53 {
   widgetClassPath=&TextBlock
   parent=Button_Run
   widget=TextBlock_53
   properties {
    text=Run
    font {
     bForceMonospaced=false
     bMaterialIsStencil=false
     fontMaterial=null
     fontObject=&Roboto
     letterSpacing=0
     monospacedWidth=1
     size=13.5
     skewAmount=0
     typefaceFontName=Regular
     outlineSettings {
      bApplyOutlineToDropShadows=false
      bMiteredCorners=false
      bSeparateFillAlpha=false
      outlineMaterial=null
      outlineSize=0
      outlineColor=<rgba 0 0 0 1>
     }
    }
   }
   slot {class=ButtonSlot}
  }
  Button_Capture {
   widgetClassPath=&EditorUtilityButton
   parent=HorizontalBox_0
   widget=Button_Capture
   properties {toolTipText="Captures any selected asset(s) in the Content Browser and adds their package paths to the Arguments field"}
   slot {class=HorizontalBoxSlot, properties {horizontalAlignment=HAlign_Left, padding=<ltrb 2.5 5 2.5 5>}}
  }
  TechAudioToolsEditorSlateIcon_340 {
   widgetClassPath=&TechAudioToolsEditorSlateIcon
   bIsVariable=true
   parent=Button_Capture
   widget=TechAudioToolsEditorSlateIcon_340
   properties {fullIconPath=EditorStyle.ClassIcon.SceneCaptureCube}
   slot {class=ButtonSlot, properties {horizontalAlignment=HAlign_Fill, padding=<ltrb 0 0 0 0>}}
  }
  Button_Clear {
   widgetClassPath=&EditorUtilityButton
   parent=HorizontalBox_0
   widget=Button_Clear
   properties {toolTipText=Clears the Arguments field}
   slot {class=HorizontalBoxSlot, properties {horizontalAlignment=HAlign_Left, padding=<ltrb 2.5 5 5 5>}}
  }
  TechAudioToolsEditorSlateIcon {
   widgetClassPath=&TechAudioToolsEditorSlateIcon
   bIsVariable=true
   parent=Button_Clear
   widget=TechAudioToolsEditorSlateIcon
   properties {fullIconPath=EditorStyle.ContentReference.Clear}
   slot {class=ButtonSlot, properties {horizontalAlignment=HAlign_Fill, padding=<ltrb 0 0 0 0>}}
  }
  TextBlock_1 {
   widgetClassPath=&TextBlock
   parent=HorizontalBox_0
   widget=TextBlock_1
   properties {
    text=Arguments:
    colorAndOpacity {colorUseRule=UseColor_Specified, specifiedColor=<rgba 0.5271 0.5271 0.5271 1>}
    font {
     bForceMonospaced=false
     bMaterialIsStencil=false
     fontMaterial=null
     fontObject=&Roboto
     letterSpacing=0
     monospacedWidth=1
     size=9
     skewAmount=0
     typefaceFontName=Regular
     outlineSettings {
      bApplyOutlineToDropShadows=false
      bMiteredCorners=false
      bSeparateFillAlpha=false
      outlineMaterial=null
      outlineSize=0
      outlineColor=<rgba 0 0 0 1>
     }
    }
   }
   slot {class=HorizontalBoxSlot, properties {horizontalAlignment=HAlign_Center, verticalAlignment=VAlign_Center, padding=<ltrb 2.5 5 2.5 5>}}
  }
  Arguments {
   widgetClassPath=&EditorUtilityEditableTextBox
   bIsVariable=true
   parent=HorizontalBox_0
   widget=Arguments
   properties {toolTipText=List of .py scripts that live in the same directory as the this Widget}
   slot {class=HorizontalBoxSlot, properties {padding=<ltrb 2.5 5 5 5>, size {sizeRule=Fill, value=0.5}}}
  }
  HorizontalBox_1 {widgetClassPath=&HorizontalBox, parent=VerticalBox_215, widget=HorizontalBox_1, slot {class=VerticalBoxSlot}}
  Metadata {
   widgetClassPath=/Script/Blutility.EditorUtilityMultiLineEditableTextBox
   bIsVariable=true
   parent=HorizontalBox_1
   widget=Metadata
   properties {bIsReadOnly=true}
   slot {class=HorizontalBoxSlot, properties {padding=<ltrb 5 5 5 5>, size {sizeRule=Fill, value=1}}}
  }
 ]
}
